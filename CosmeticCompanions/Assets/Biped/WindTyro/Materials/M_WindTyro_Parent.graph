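# Parent material for the WindTyro cosmetic companion, wrapping a custom companion color function between the shared character-parent PreFX and PostFX functions.
# The body mask's red channel, scaled by a global glow parameter, marks glow areas that darken base color and add emissive.
# An optional time-based sawtooth pulse, phase-offset by the green channel, varies that glow.
# Inverted Fresnel plus a bias weights the glow toward camera-facing surfaces.
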Material /CosmeticCompanions/Assets/Biped/WindTyro/Materials/M_WindTyro_Parent
graph {
 properties {
  bAutomaticallySetUsageInEditor=false
  bUseMaterialAttributes=true
  bUsedWithMeshDeformer=true
  bUsedWithSkeletalMesh=true
  bUsedWithStaticLighting=true
  usedShadingModels=Default Lit
 }
 nodes [
  Frac(None=Time())
  n1 = MaterialFunctionCall() {materialFunction=&MF_CharacterParent_2023_PreFX}
  n2 = GetMaterialAttributes(None=MaterialFunctionCall!(Material=@n1.Result, Tech Art Mask=@n1.Result) {materialFunction=/CosmeticCompanions/Materials/MaterialFunctions/MF_FN_Companion_Custom_Color}.Base Color) {attributeGetTypes [#1, #2]}
  MaterialFunctionCall(
   Material=SetMaterialAttributes!(MaterialAttributes=@n2.MaterialAttributes, Emissive Color=NamedRerouteUsage(), Base Color=NamedRerouteUsage()) {attributeSetTypes [#1, #2]}
   M=@n1.M
   Pre Skinned Local Position=@n1.M
   Distance Blend=@n1.M
   Fresnel=@n1.M
  ) {materialFunction=&MF_CharacterParent_2023_PostFX}
  n3 = TextureSample() {samplerType=SAMPLERTYPE_Masks, texture=/CosmeticCompanions/Assets/Biped/WindTyro/Textures/T_WindTyro_Body_TA}
  n4 = Multiply(A=Saturate(None=ScalarParameter() {group=WindTyro, parameterName=WT_GlobalGlow}), B=@n3.R)
  n5 = Fresnel(ExponentIn=ScalarParameter() {defaultValue=0.5, parameterName=EXP}) {baseReflectFraction=0, exponent=0.5}
  n6 = ScalarParameter() {defaultValue=2, parameterName=POW}
  Multiply(A=@n4, B=OneMinus(None=Power(Base=@n5, Exp=@n6) {constExponent=11}))
  NamedRerouteDeclaration(None=@n2.Emissive Color) {name=Emissive_IN, nodeColor=<rgba 0.6641 0 1 1>}
  n7 = Time()
  n8 = ScalarParameter() {defaultValue=1, group=WindTyro, parameterName=WT_GlowRandomSpeedStart}
  n9 = Add(A=Frac(None=Multiply(A=@n7, B=@n8)), B=@n3.G)
  n10 = Power(Base=OneMinus(None=Frac(None=@n9)))
  n11 = ScalarParameter() {group=WindTyro, parameterName=WT_GlowRandom}
  n12 = Multiply(
   A=Multiply(A=NamedRerouteUsage(), B=@n4)
   B=Multiply(A=ScalarParameter() {defaultValue=6, parameterName=WT_GlowStrength}, B=LinearInterpolate(B=@n10, Alpha=@n11) {constA=1})
  )
  n13 = OneMinus(None=@n5)
  n14 = Saturate(None=Add(A=Power(Base=@n13), B=ScalarParameter() {defaultValue=0.4, parameterName=BIAS}) {constB=0.3})
  NamedRerouteDeclaration(None=Add(A=NamedRerouteUsage(), B=Multiply(A=@n12, B=@n14))) {name=Emissive_OUT, nodeColor=<rgba 0 1 0.3438 1>}
  NamedRerouteDeclaration(None=@n2.Base Color) {name=Basecolor_IN, nodeColor=<rgba 1 0.0234 0 1>}
  NamedRerouteDeclaration(None=LinearInterpolate(A=NamedRerouteUsage(), Alpha=@n4) {constB=0}) {name=Basecolor_OUT, nodeColor=<rgba 0 0.2969 1 1>}
  ScalarParameter() {defaultValue=1, parameterName=DEBUG}
 ]
}
