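# Animation Blueprint for the WindTyro biped cosmetic companion, retargeting its pose from another mesh.
# The output blends by inherited bUseReverseRetarget between the emote IK retargeter (true) and the standard one (false).
# A constraint chain tying attach_emote_02/03 to dyn_topper_01/02 when that flag is false is not wired into the output.
# The event graph holds no working logic.
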AnimBlueprint /CosmeticCompanions/Assets/Biped/WindTyro/Meshes/WindTyro_Biped_AnimBP
target_skeleton=&/WindTyro_Biped_Skeleton
graphs [
 AnimGraph {nodes [
   n1 = NOT Boolean(A=Get bUseReverseRetarget() -> (bUseReverseRetarget:bool))
   n2 = "Constraint - attach_emote_02 (P)"!(ComponentPose=Local To Component(), bAlphaBoolEnabled:bool=@n1, ConstraintWeights_0=1) -> (Pose) {alphaInputType=Bool, boneToModify {boneName=attach_emote_02}, constraintSetup [{targetBone {boneName=dyn_topper_01}}], constraintWeights [1]}
   n3 = "Constraint - attach_emote_03 (P)"!(ComponentPose=@n2, bAlphaBoolEnabled:bool=@n1, ConstraintWeights_0=1) -> (Pose) {alphaInputType=Bool, boneToModify {boneName=attach_emote_03}, constraintSetup [{targetBone {boneName=dyn_topper_02}}], constraintWeights [1]}
   Component To Local(ComponentPose=@n3)
   Output Pose(Result=Blend Poses by bool(
     BlendPose_0=Retarget Pose From Mesh!() -> (Pose:Pose Link Structure) {is=title AnimGraphNode_RetargetPoseFromMesh, iKRetargeterAsset=&/IK_Retargeter_WindTyro_Emote}
     BlendPose_1=Retarget Pose From Mesh() {iKRetargeterAsset=&/IK_Retargeter_WindTyro}
     BlendTime_0=0.2
     BlendTime_1=0.2
     bActiveValue=Get bUseReverseRetarget()
    ))
  ]}
 EventGraph {nodes [TryGetPawnOwner()]}
]
variables [
 Custom Pin Bones Profile {
  type=RetargetProfile
  default {
   bApplyChainSettings=false
   bApplyGlobalSettings=false
   bApplyRootSettings=false
   bApplySourceRetargetPose=false
   bApplyTargetRetargetPose=false
   bForceAllIKOff=false
   sourceRetargetPoseName=None
   targetRetargetPoseName=None
   chainSettings {}
   globalSettings {
    bCopyBasePose=false
    bEnableFK=true
    bEnableIK=true
    bEnablePost=true
    bEnableRoot=true
    bWarping=false
    copyBasePoseRoot=None
    directionChain=None
    directionSource=Goals
    forwardDirection=Y
    sidewaysOffset=0
    sourceScaleFactor=1
    warpForwards=1
    warpSplay=1
   }
   retargetOpProfiles []
   rootSettings {
    affectIKHorizontal=1
    affectIKVertical=0
    blendToSource=0
    rotationAlpha=1
    scaleHorizontal=1
    scaleVertical=1
    translationAlpha=1
    blendToSourceWeights=<1 1 1>
    rotationOffset=<rot 0 0 0>
    translationOffset=<0 0 0>
   }
  }
 }
 IKRetargeter Asset {type=IKRetargeter, default=&IK_Retargeter_WindTyro}
]
