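# Post-process animation blueprint for the WindTyro biped cosmetic companion, correcting facial, accessory and feather bones on the incoming pose.
# The input pose drives a facial PoseAsset blend, then passes through chained bone constraints in component space.
# The accessory attach bone follows a weighted spine and neck mix, dominated by spine_03, with inverted AnimateAcc curve alpha so the constraint fades as the curve rises.
# Two featherB bones follow their matching dyn_topper bones; the event graph holds no logic.
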AnimBlueprint /CosmeticCompanions/Assets/Biped/WindTyro/Meshes/WindTyro_Biped_Post_AnimBP
target_skeleton=&/WindTyro_Biped_Skeleton
graphs [
 AnimGraph {nodes [
   n1 = Input Pose()
   n2 = "Constraint - attach_accessory (PPPP)"!(
    ComponentPose=Local To Component(LocalPose=WindTyro_Biped_Head_Facial_Poses_PoseAsset!(SourcePose:Pose Link Structure=@n1) -> (Pose:Pose Link Structure) {
      is=title AnimGraphNode_PoseBlendNode
      blendOption=Linear
      poseAsset=&/WindTyro_Biped_Head_Facial_Poses_PoseAsset
     })
    AlphaCurveName:Name=AnimateAcc
    ConstraintWeights_0=0.849
    ConstraintWeights_1:float=0.1443
    ConstraintWeights_2:float=0.0053
    ConstraintWeights_3:float=0.0014
   ) -> (Pose) {
    alphaInputType=Curve
    alphaBoolBlend {blendOption=Linear}
    alphaScaleBiasClamp {bias=1, scale=-1}
    boneToModify {boneName=attach_accessory}
    constraintSetup [{targetBone {boneName=spine_03}}, {targetBone {boneName=neck_01}}, {targetBone {boneName=neck_02}}, {targetBone {boneName=spine_02}}]
    constraintWeights [1,1,1,1]
   }
   n3 = "Constraint - featherB_01 (P)"(ComponentPose=@n2, Alpha=1, ConstraintWeights_0=1) {
    alphaBoolBlend {blendOption=Linear}
    boneToModify {boneName=featherB_01}
    constraintSetup [{targetBone {boneName=dyn_topper_01}}]
    constraintWeights [1]
   }
   n4 = "Constraint - featherB_02 (P)"(ComponentPose=@n3, Alpha=1, ConstraintWeights_0=1) {
    alphaBoolBlend {blendOption=Linear}
    boneToModify {boneName=featherB_02}
    constraintSetup [{targetBone {boneName=dyn_topper_02}}]
    constraintWeights [1]
   }
   Output Pose(Result=Component To Local(ComponentPose=@n4))
  ]}
 EventGraph {nodes [
   TryGetPawnOwner()
   Event BlueprintUpdateAnimation()
  ]}
]
variables []
